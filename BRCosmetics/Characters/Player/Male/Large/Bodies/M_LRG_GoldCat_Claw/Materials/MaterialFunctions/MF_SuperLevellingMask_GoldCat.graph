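# Material function that builds the GoldCat Claw's super-levelling mask, packing what appear to be body and edge masks into a float3.
# Time since a Blueprint timestamp animates the level step, normalized and remapped through a curve atlas to move and grow a sphere mask along three points.
# Pre-skinned position and world-aligned noise break up the transition, and a paw SDF texture shapes it.
# The quantized-progress branch feeds nothing.
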MaterialFunction /BRCosmetics/Characters/Player/Male/Large/Bodies/M_LRG_GoldCat_Claw/Materials/MaterialFunctions/MF_SuperLevellingMask_GoldCat
graph {nodes [
  FunctionOutput(None=MaterialFunctionCall(X=NamedRerouteUsage(), Y=NamedRerouteUsage(), Z=Constant()) {materialFunction=&MakeFloat3}.Result) {outputName=SuperLevelling_Mask}
  n1 = ScalarParameter() {group=SuperLevelling_Mask, parameterName=SL_CurrentLevel, sortPriority=3}
  n2 = Subtract(A=Time(), B=ScalarParameter() {defaultValue=-99999, group=SuperLevelling_Mask, parameterName=SL_GameTimeFromBP, sortPriority=100})
  n3 = Add(A=ScalarParameter() {defaultValue=2.5, group=SuperLevelling_Mask, parameterName=SL_Animation_Duration, sortPriority=101}) {constB=0}
  n4 = LinearInterpolate(A=Subtract(A=@n1), B=@n1, Alpha=Saturate(None=Divide(A=@n2, B=@n3)))
  n5 = ScalarParameter() {group=SuperLevelling_Mask, parameterName=SL_StartAtLevel, sortPriority=3}
  n6 = ScalarParameter() {defaultValue=3, group=SuperLevelling_Mask, parameterName=SL_StopAtLevel, sortPriority=3}
  n7 = Divide(A=Subtract(A=Clamp(None=@n4, Min=@n5, Max=@n6), B=@n5), B=Subtract(A=@n6, B=@n5))
  n8 = Max(A=Ceil(None=ScalarParameter() {defaultValue=1, group=SuperLevelling_Mask, parameterName=SL_QuantizationSteps, sliderMax=100, sliderMin=1}))
  n9 = Multiply(A=@n7, B=@n8)
  n10 = Subtract(A=@n8)
  LinearInterpolate(B=Divide(A=Floor(None=@n9), B=@n8), Alpha=Saturate(None=Ceil(None=@n10)))
  n11 = NamedRerouteUsage()
  n12 = NamedRerouteUsage()
  n13 = CurveAtlasRowParameter(CurveTime=@n7) {
   atlas=&AthenaCurveAtlas01
   curve=/Game/Athena/LinearColorCurves/Curve_SuperlevelMask_Summer
   defaultValue=289
   group=SuperLevelling_Mask
   parameterName=SL_SuperLevelMaskCurve
  }
  n14 = ChannelMaskParameter(None=@n13) {group=SuperLevelling_Mask, parameterName=SL_SuperLevelCurveChannel}
  n15 = Power(Base=@n14.R, Exp=ScalarParameter() {defaultValue=1, group=SuperLevelling_Mask, parameterName=SL_MaskPos_Exponent})
  n16 = MaterialFunctionCall(
   A=VectorParameter() {group=SuperLevelling_Mask, parameterName=SL_MaskPos_01, defaultValue=<rgba -38 4 100 1>}.RGB
   B=VectorParameter() {group=SuperLevelling_Mask, parameterName=SL_MaskPos_02, defaultValue=<rgba -1 -0.0001 163 1>}.RGB
   C=VectorParameter() {group=SuperLevelling_Mask, parameterName=SL_MaskPos_03, defaultValue=<rgba -12 0 144 1>}.RGB
   Alpha=@n15
  ) {materialFunction=&Lerp_3Color}
  n17 = LinearInterpolate(
   A=ScalarParameter() {defaultValue=10, group=SuperLevelling_Mask, parameterName=SL_StartRadius}
   B=ScalarParameter() {defaultValue=227, group=SuperLevelling_Mask, parameterName=SL_EndRadius}
   Alpha=@n14.R
  )
  n18 = Multiply(A=@n17, B=ScalarParameter() {defaultValue=1.3, parameterName=PawsRadius})
  n19 = ScalarParameter() {parameterName=PawsGradHardness}
  n20 = Add(
   A=SmoothStep(Value=SphereMask(A=@n12, B=@n16.Result, Radius=@n18, Hardness=@n19) {hardnessPercent=50}) {constMax=0.2}
   B=Subtract(
    A=TextureSample() {texture=/BRCosmetics/Characters/Player/Male/Large/Bodies/M_LRG_GoldCat_Claw/Textures/T_GoldCat_Paws_SDF_Mask}.G
    B=Constant() {r=0.75}
   ) {constB=0.5}
  )
  n21 = LinearInterpolate(
   A=@n11
   B=LinearInterpolate(A=@n20, Alpha=@n11)
   Alpha=NamedRerouteDeclaration(None=ScalarParameter() {defaultValue=1, group=SuperLevelling_Mask, parameterName=SL_Transition_UsePaw}) {name=UsePaw, nodeColor=<rgba 0 0.1797 1 1>}
  )
  n22 = SmoothStep(
   Min=ScalarParameter() {defaultValue=0.4, group=SuperLevelling_Mask, parameterName=HardMask_Min, sortPriority=100}
   Max=ScalarParameter() {defaultValue=0.45, group=SuperLevelling_Mask, parameterName=HardMask_Max, sortPriority=100}
   Value=Saturate(None=@n21)
  )
  NamedRerouteDeclaration(None=Saturate(None=StaticSwitchParameter(True=@n22, False=@n21) {defaultValue=true, group=SuperLevelling_Mask, parameterName=SL_UseHardEdgedMask})) {name=BodyMask, nodeColor=<rgba 0 1 0.6016 1>}
  n23 = SmoothStep(
   Min=ScalarParameter() {defaultValue=0.39, group=SuperLevelling_Mask, parameterName=HardMask_Min_Emissive, sortPriority=101}
   Max=ScalarParameter() {defaultValue=0.47, group=SuperLevelling_Mask, parameterName=HardMask_Max_Emissive, sortPriority=101}
   Value=NamedRerouteUsage()
  )
  n24 = SmoothStep(
   Min=ScalarParameter() {defaultValue=1, group=SuperLevelling_Mask, parameterName=HardMask_Min_Emissive_Paw, sortPriority=101}
   Max=ScalarParameter() {defaultValue=0.3, group=SuperLevelling_Mask, parameterName=HardMask_Max_Emissive_Paw, sortPriority=101}
   Value=Saturate(None=LinearInterpolate(B=@n20, Alpha=@n23) {constB=0})
  )
  n25 = OneMinus(None=@n24)
  n26 = Multiply(A=If(A=@n24, B=Constant() {r=0.9}, "A > B"=@n25, "A == B"=@n25, "A < B"=@n24)) {constB=0.1}
  n27 = OneMinus(None=@n23)
  n28 = If(A=@n23, B=Constant() {r=0.9}, "A > B"=@n27, "A == B"=@n27, "A < B"=@n23)
  NamedRerouteDeclaration(None=LinearInterpolate(A=@n28, B=Add(A=@n28, B=@n26), Alpha=NamedRerouteUsage())) {name=EdgeMask, nodeColor=<rgba 1 0.5625 0 1>}
  n29 = TextureObjectParameter() {
   group=SuperLevelling_Mask
   parameterName=SL_TransitionNoise
   texture=/Game/Characters/Player/Male/Medium/Bodies/M_MED_Viper/Textures/Glitch/T_SoftNoise
  }
  n30 = LocalPosition() {localOrigin=InstancePreSkinning}
  n31 = VectorParameter() {group=SuperLevelling_Mask, parameterName=SL_Backpack_PreskinnedPositionOffset, defaultValue=<rgba 0 0 0 1>}
  n32 = Add(A=FunctionInput(Preview=VertexInterpolator(VS=@n30.XYZ).PS) {bUsePreviewValueAsDefault=true, inputName=Local Position}, B=@n31.RGB)
  n33 = VertexInterpolator(VS=PreSkinnedNormal())
  n34 = ChannelMaskParameter(None=MaterialFunctionCall(
    TextureObject=@n29
    TextureSize=VectorParameter() {group=SuperLevelling_Mask, parameterName=SL_TransitionNoise_Tiling, defaultValue=<rgba 40 10 42 1>}.RGB
    WorldPosition=NamedRerouteDeclaration(None=@n32) {name=PreSkinnedPositionOffset, nodeColor=<rgba 1 0.8906 0 1>}
    World Space Normal=FunctionInput(Preview=@n33.PS) {bUsePreviewValueAsDefault=true, inputName=Local Normal}
   ) {materialFunction=/Game/Weapons/WeaponSkins/Wraps/MaterialFunctions/WorldAlignedTexture}.XYZ Texture) {group=SuperLevelling_Mask, maskChannel=Green, parameterName=SL_TransitionNoiseChannel, defaultValue=<rgba 0 1 0 0>}
  NamedRerouteDeclaration(None=@n34.R) {name=TransitionNoise, nodeColor=<rgba 0 1 0.5078 1>}
  n35 = SphereMask(
   A=NamedRerouteUsage()
   B=@n16.Result
   Radius=@n17
   Hardness=ScalarParameter() {defaultValue=0.5, group=SuperLevelling_Mask, parameterName=SL_CoverMaskHardness}
  )
  n36 = Multiply(
   A=NamedRerouteUsage()
   B=LinearInterpolate(
    A=ScalarParameter() {defaultValue=0.5, group=SuperLevelling_Mask, parameterName=SL_Transition_Falloff_Start}
    B=ScalarParameter() {defaultValue=0.1, group=SuperLevelling_Mask, parameterName=SL_Transition_Falloff_End}
    Alpha=@n15
   )
  )
  NamedRerouteDeclaration(None=Saturate(None=Subtract(A=@n35, B=@n36))) {name=TransitionGradient, nodeColor=<rgba 1 0.1875 0 1>}
 ]}
